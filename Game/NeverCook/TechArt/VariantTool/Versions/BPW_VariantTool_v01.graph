EditorUtilityWidgetBlueprint /Game/NeverCook/TechArt/VariantTool/Versions/BPW_VariantTool_v01
graphs [EventGraph {nodes [
   n1 = GetDataTableRowNames(Table=/VehicleCosmetics/DataTables/DT_VehicleCosmetics_Painted)
   > n2 = For Each Loop(Array:Array of Names=@n1) -> (Array Element:Name)
   > PrintString(InString="To String (Name)"(InName=@n2))
   "On Clicked (Button_CreatePaintedVariants)"() {is=title K2Node_ComponentBoundEvent}
   > InitSelectionVars()
   > GetSelectedAsset()
   > NameAndGenerateVariants()
   "On Clicked (Button_DebugInfo)"() {is=title K2Node_ComponentBoundEvent}
   > Debug_ID_Data()
   GetSelectedAsset() {is=title K2Node_CustomEvent}
   > Clear(TargetArray:Array of Object References=Get SelectedAssets() -> (SelectedAssets:Array of Object References))
   > n3 = GetSelectedAssets()
   > n4 = Branch(Condition=Is Empty!(TargetArray:Array of Object References=@n3) -> (ReturnValue:bool))
   > PrintString(InString=WARNING: Nothing is selected. Please select an Vehicle ID to proceed., TextColor=(R=1,G=0,B=0.0646,A=1), Duration=5)
   n5 = Branch(
    execute=@n4.else
    Condition=IsA!(
     Object:Object Reference="Get (a copy)"(Array:Array of Object References=@n3) -> (Output:Object Reference)
     ObjectClass:Object Class Reference=/Script/FortniteGame.FortItemDefinition
    ) -> (ReturnValue:bool)
   )
   > PrintString(InString=SUCCESS: Correct ID Selected., TextColor=(R=0,G=0.66,B=0.126,A=1), Duration=5)
   > n6 = Sequence()
   > n7 = GetAssetTools()
   > Set AssetTools(AssetTools:Asset Tools Interface=@n7) -> (Output_Get:Asset Tools Interface)
   n8 = "Get (a copy)"(Array=@n3)
   n9 = GetCurrentContentBrowserPath!() -> (OutPath:String, ReturnValue:bool)
   > n10 = LoadAsset(AssetPath=Append!(A=@n9, B=/, C:String=GetDisplayName(Object=@n8)) -> (ReturnValue))
   > Set SelectedAssets(SelectedAssets:Array of Object References=Make Array("[0]":Object Reference=@n10) -> (Array:Array of Object References)) -> (Output_Get:Array of Object References)
   Set CurrentSelectionPath(execute=@n6.then_1, CurrentSelectionPath:String=@n9) -> (Output_Get:String)
   n11 = GetDisplayName(Object="Get (a copy)"(Array=Get SelectedAssets()))
   > n12 = Set SelectedIDName(SelectedIDName:String=@n11) -> (Output_Get:String)
   > Set SelectedID_PathAndName(SelectedID_PathAndName:String=Append(A=@n9, B=/, C=@n11)) -> (Output_Get:String)
   > Set VehicleName(VehicleName:String=RightChop!(SourceString:String=@n12, Count:int=8) -> (ReturnValue:String)) -> (Output_Get:String)
   PrintString(
    execute=@n5.else
    InString=WARNING: Invalid Vehicle ID Class. Please select an Vehicle ID to proceed.
    TextColor=(R=1,G=0,B=0.0646,A=1)
    Duration=5
   )
   InitSelectionVars() {is=title K2Node_CustomEvent}
   > Clear(TargetArray=Get SelectedAssets())
   > Set VehicleName()
   > Set SelectedIDName()
   > Set CurrentSelectionPath()
   > Set SelectedID_PathAndName()
   NameAndGenerateVariants() {is=title K2Node_CustomEvent}
   > n13 = GetDataTableRowNames(Table=&DT_VehicleCosmetics_Painted)
   > n14 = Set PaintNames(PaintNames:Array of Names=@n13) -> (Output_Get:Array of Names)
   > n15 = For Each Loop(Array=@n14)
   > n16 = GetAssetTools()
   n17 = "To String (Name)"(InName=@n15)
   > n18 = CreateAsset(
    self=@n16
    AssetName=Append!(A=CarBody_, B=Get VehicleName() -> (VehicleName:String), C=_, D:String=@n17) -> (ReturnValue)
    PackagePath=Append!(A=GetCurrentContentBrowserPath(), B=/, C=Variants) -> (ReturnValue)
    AssetClass=Get Class(Object=/Game/NeverCook/TechArt/VariantTool/_TemplateFiles/PaintedVariant_Template) -> (ReturnValue:Object Class Reference)
   )
   > Execute Python Script!(
    execute
    PythonScript="import unreal\r\n\r\nasset = my_asset\r\ncolor_array = my_color\r\nselected_assets = sel_asset_input\r\n\r\ntagChannel = unreal.FortniteEditorLibrary.request_gameplay_tag(\"Cosmetics.Variant.Channel.Vehicle.Painted\", True)\r\ntagName = unreal.FortniteEditorLibrary.request_gameplay_tag(\"Cosmetics.Variant.Property.Vehicle.Painted.\" + color_array, True)\r\n\r\n\r\nCosmeticItemName = unreal.EditorAssetLibrary.load_asset(selected_assets)\r\n\r\nasset.set_editor_property(\"VariantChannelTag\", tagChannel)\r\nasset.set_editor_property(\"VariantNameTag\", tagName)\r\nasset.set_editor_property(\"cosmetic_item\", CosmeticItemName)\r\n\r\nprint (tagName)"
    my_asset:Object Reference=@n18
    my_color:String=@n17
    SelAssetInput:String=Get SelectedID_PathAndName() -> (SelectedID_PathAndName:String)
   ) -> (then, ReturnValue) {inputs [my_asset, my_color, SelAssetInput]}
   Debug_ID_Data() {is=title K2Node_CustomEvent}
   > InitSelectionVars() {is=title K2Node_CallFunction}
   > GetSelectedAsset() {is=title K2Node_CallFunction}
   > PrintString(InString=Append!(A="Current Selection is: ", B=Get SelectedIDName() -> (SelectedIDName:String)) -> (ReturnValue))
   > PrintString(InString=Append(A="Path of Selection: ", B=Get CurrentSelectionPath() -> (CurrentSelectionPath:String)))
   > PrintString(InString=Append(A="Path and File in Variable: ", B=Get SelectedID_PathAndName()))
   > PrintString(InString=Append(A="Vehicle Name From File: ", B=Get VehicleName()), TextColor=(R=0,G=0.66,B=0.126,A=1), Duration=5)
   Event Construct()
   > InitSelectionVars()
  ]}]
variables [
 AssetTools {default=None}
 CurrentSelectionPath {type=string, default=""}
 PaintNames {type="array<string>", default []}
 SelectedAssets {type="array<Object>", default [null]}
 SelectedID_PathAndName {type=string, default=""}
 SelectedIDName {type=string, default=""}
 VehicleName {type=string, default=VEHICLENAME}
 VTID_Template {type=FortVehicleCosmeticsItemDefinition_Body, default=null}
]
widgets {
 info {rootWidgetClass=&VerticalBox, widgetCount=17}
 widgets [
  VerticalBox_82 {widgetClassPath=&VerticalBox, widget=VerticalBox_82}
  ScaleBox_163 {widgetClassPath=&ScaleBox, parent=VerticalBox_82, widget=ScaleBox_163, slot {class=VerticalBoxSlot}}
  VerticalBox_117 {
   widgetClassPath=&VerticalBox
   parent=ScaleBox_163
   widget=VerticalBox_117
   slot {class=ScaleBoxSlot, properties {horizontalAlignment=HAlign_Fill, verticalAlignment=VAlign_Fill}}
  }
  HorizontalBox_145 {
   widgetClassPath=&HorizontalBox
   parent=VerticalBox_117
   widget=HorizontalBox_145
   slot {class=VerticalBoxSlot, properties {padding=<ltrb 10 10 10 10>}}
  }
  IMG_RL_Logo {
   widgetClassPath=&Image
   bIsVariable=true
   parent=HorizontalBox_145
   widget=IMG_RL_Logo
   properties {brush {
     bIsDynamicallyLoaded=false
     drawAs=Image
     imageType=NoImage
     mirroring=NoMirror
     resourceName=None
     resourceObject=/Game/NeverCook/TechArt/VehicleCosmeticsTools/VehicleCosmetics_Picker/Icons/UI_RL_Icon
     tiling=NoTile
     imageSize=<64 64>
     margin=<ltrb 0 0 0 0>
     outlineSettings {
      bUseBrushTransparency=false
      roundingType=HalfHeightRadius
      width=0
      color {colorUseRule=UseColor_Specified, specifiedColor=<rgba 0 0 0 0>}
      cornerRadii=<0 0 0 0>
     }
     tintColor {colorUseRule=UseColor_Specified, specifiedColor=<rgba 1 1 1 1>}
     uVRegion {bIsValid=false, max=<0 0>, min=<0 0>}
    }}
   slot {class=HorizontalBoxSlot, properties {horizontalAlignment=HAlign_Left, verticalAlignment=VAlign_Center}}
  }
  Spacer_55 {widgetClassPath=&Spacer, parent=HorizontalBox_145, widget=Spacer_55, properties {size=<15 1>}, slot {class=HorizontalBoxSlot}}
  TXT_HeaderTitle {
   widgetClassPath=&TextBlock
   parent=HorizontalBox_145
   widget=TXT_HeaderTitle
   properties {
    text=Variant Creator
    font {
     bForceMonospaced=false
     bMaterialIsStencil=false
     fontMaterial=null
     fontObject=null
     letterSpacing=0
     monospacedWidth=1
     size=48
     skewAmount=0
     typefaceFontName=Default
     outlineSettings {
      bApplyOutlineToDropShadows=false
      bMiteredCorners=false
      bSeparateFillAlpha=false
      outlineMaterial=null
      outlineSize=0
      outlineColor=<rgba 0 0 0 1>
     }
    }
   }
   slot {class=HorizontalBoxSlot, properties {horizontalAlignment=HAlign_Left, verticalAlignment=VAlign_Center}}
  }
  HorizontalBox_322 {widgetClassPath=&HorizontalBox, parent=VerticalBox_117, widget=HorizontalBox_322, slot {class=VerticalBoxSlot}}
  Button_DebugInfo {
   widgetClassPath=&EditorUtilityButton
   parent=HorizontalBox_322
   widget=Button_DebugInfo
   properties {toolTipText="Prints out data related to the selected Vehicle ID. If it's not a FortItemDefinition Class, it will let you know. To create variants ensure you have"}
   slot {class=HorizontalBoxSlot, properties {verticalAlignment=VAlign_Bottom, padding=<ltrb 10 10 10 10>}}
  }
  TXT_DebugInfo {
   widgetClassPath=&TextBlock
   parent=Button_DebugInfo
   widget=TXT_DebugInfo
   properties {
    text=Print Debug Info
    colorAndOpacity {colorUseRule=UseColor_Specified, specifiedColor=<rgba 0.526 0.526 0.526 1>}
    font {
     bForceMonospaced=false
     bMaterialIsStencil=false
     fontMaterial=null
     fontObject=null
     letterSpacing=0
     monospacedWidth=1
     size=24
     skewAmount=0
     typefaceFontName=RLSmall-Regular
     outlineSettings {
      bApplyOutlineToDropShadows=false
      bMiteredCorners=false
      bSeparateFillAlpha=false
      outlineMaterial=null
      outlineSize=0
      outlineColor=<rgba 0 0 0 1>
     }
    }
   }
   slot {class=ButtonSlot}
  }
  Button_CreatePaintedVariants {
   widgetClassPath=&EditorUtilityButton
   parent=HorizontalBox_322
   widget=Button_CreatePaintedVariants
   slot {class=HorizontalBoxSlot, properties {padding=<ltrb 10 10 10 10>}}
  }
  TXT_CreatePaintedVariants {
   widgetClassPath=&TextBlock
   parent=Button_CreatePaintedVariants
   widget=TXT_CreatePaintedVariants
   properties {
    text=Create Variants
    colorAndOpacity {colorUseRule=UseColor_Specified, specifiedColor=<rgba 0.526 0.526 0.526 1>}
    font {
     bForceMonospaced=false
     bMaterialIsStencil=false
     fontMaterial=null
     fontObject=null
     letterSpacing=0
     monospacedWidth=1
     size=24
     skewAmount=0
     typefaceFontName=RLSmall-Regular
     outlineSettings {
      bApplyOutlineToDropShadows=false
      bMiteredCorners=false
      bSeparateFillAlpha=false
      outlineMaterial=null
      outlineSize=0
      outlineColor=<rgba 0 0 0 1>
     }
    }
   }
   slot {class=ButtonSlot}
  }
  HBox_VehicleName {
   widgetClassPath=&HorizontalBox
   parent=VerticalBox_117
   widget=HBox_VehicleName
   slot {class=VerticalBoxSlot, properties {padding=<ltrb 10 0 10 10>}}
  }
  TXT_VehicleName {
   widgetClassPath=&TextBlock
   parent=HBox_VehicleName
   widget=TXT_VehicleName
   properties {
    text=VEHICLE NAME
    colorAndOpacity {colorUseRule=UseColor_Specified, specifiedColor=<rgba 0.526 0.526 0.526 1>}
    font {
     bForceMonospaced=false
     bMaterialIsStencil=false
     fontMaterial=null
     fontObject=null
     letterSpacing=0
     monospacedWidth=1
     size=24
     skewAmount=0
     typefaceFontName=RLSmall-Regular
     outlineSettings {
      bApplyOutlineToDropShadows=false
      bMiteredCorners=false
      bSeparateFillAlpha=false
      outlineMaterial=null
      outlineSize=0
      outlineColor=<rgba 0 0 0 1>
     }
    }
   }
   slot {class=HorizontalBoxSlot, properties {horizontalAlignment=HAlign_Left, size {sizeRule=Fill, value=0.5}}}
  }
  TXT_VehName_Seperator {
   widgetClassPath=&TextBlock
   parent=HBox_VehicleName
   widget=TXT_VehName_Seperator
   properties {
    text="====="
    colorAndOpacity {colorUseRule=UseColor_Specified, specifiedColor=<rgba 0.526 0.526 0.526 1>}
    font {
     bForceMonospaced=false
     bMaterialIsStencil=false
     fontMaterial=null
     fontObject=null
     letterSpacing=0
     monospacedWidth=1
     size=13.5
     skewAmount=0
     typefaceFontName=RLSmall-Regular
     outlineSettings {
      bApplyOutlineToDropShadows=false
      bMiteredCorners=false
      bSeparateFillAlpha=false
      outlineMaterial=null
      outlineSize=0
      outlineColor=<rgba 0 0 0 1>
     }
    }
   }
   slot {class=HorizontalBoxSlot, properties {horizontalAlignment=HAlign_Left, size {sizeRule=Fill, value=0.5}}}
  }
  TextBlock_311 {
   widgetClassPath=&TextBlock
   parent=HBox_VehicleName
   widget=TextBlock_311
   properties {
    text=VEHICLENAME
    font {
     bForceMonospaced=false
     bMaterialIsStencil=false
     fontMaterial=null
     fontObject=null
     letterSpacing=0
     monospacedWidth=1
     size=24
     skewAmount=0
     typefaceFontName=RLSmall-Regular
     outlineSettings {
      bApplyOutlineToDropShadows=false
      bMiteredCorners=false
      bSeparateFillAlpha=false
      outlineMaterial=null
      outlineSize=0
      outlineColor=<rgba 0 0 0 1>
     }
    }
   }
   slot {class=HorizontalBoxSlot, properties {horizontalAlignment=HAlign_Right}}
  }
  Spacer_104 {widgetClassPath=&Spacer, parent=VerticalBox_117, widget=Spacer_104, properties {size=<1 100>}, slot {class=VerticalBoxSlot}}
 ]
}
